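# Material function for the SheerWill Blot FrothRain skin that builds a spreading body mask and glowing edge color for an emote.
# Sphere masks on pre-skinned local position, distorted by noise textures, cover arms, legs, feet and head and merge through Max into OutMask.
# The SheerWillEmoteAnimation scalar drives saturated stage ramps alongside staged radius parameters, with NoHelmet variants for the head.
# EmoteEdgeColor tints the mask's SmoothStep edge, scaled by a sine of the animation.
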MaterialFunction /BRCosmetics/Characters/Player/Male/Medium/Bodies/M_MED_SheerWill_Blot/Skins/FrothRain/Materials/MF/MF_SheerWill_EmoteMask
graph {nodes [
  n1 = TextureSample(UVs=Multiply(A=TextureCoordinate(), B=Constant() {r=3.1009})) {samplerSource=SSM_Wrap_WorldGroupSettings, samplerType=SAMPLERTYPE_LinearColor, texture=/Game/Global/Textures/Default/Noise/T_CloudNoise_01}
  n2 = LinearInterpolate(Alpha=@n1.B) {constA=-1}
  n3 = Constant() {r=7.235}
  n4 = NamedRerouteUsage()
  n5 = Constant4Vector() {constant=<rgba -38.8338 0 -128.0196 1>}
  n6 = SphereMask(
   A=Add(A=Multiply(A=@n2, B=@n3), B=Add(A=@n4, B=@n5.RGB))
   B=Constant()
   Radius=LinearInterpolate(B=ScalarParameter() {defaultValue=18.2411, parameterName=Arm_L_Radius}, Alpha=NamedRerouteUsage())
   Hardness=Constant() {r=0.98}
  )
  n7 = TextureSample(UVs=Multiply(A=TextureCoordinate(), B=Constant() {r=3.1009})) {samplerSource=SSM_Wrap_WorldGroupSettings, samplerType=SAMPLERTYPE_LinearColor, texture=&T_CloudNoise_01}
  n8 = Multiply(A=LinearInterpolate(Alpha=@n7.B) {constA=-1}, B=Constant() {r=7.235})
  n9 = Add(A=NamedRerouteUsage(), B=Constant4Vector() {constant=<rgba 51.6587 0 -133.7355 1>}.RGB)
  n10 = ScalarParameter() {defaultValue=22, parameterName=Arm_R_Radius}
  n11 = NamedRerouteUsage()
  n12 = TextureSample(UVs=Multiply(A=TextureCoordinate(), B=Constant() {r=3.1009})) {samplerSource=SSM_Wrap_WorldGroupSettings, samplerType=SAMPLERTYPE_LinearColor, texture=&T_CloudNoise_01}
  n13 = Add(
   A=Multiply(A=LinearInterpolate(Alpha=@n12.B) {constA=-1}, B=Constant() {r=3.6713})
   B=Add(A=NamedRerouteUsage(), B=Constant4Vector() {constant=<rgba 89.2488 0 -54 1>}.RGB)
  )
  n14 = Constant()
  n15 = ScalarParameter() {defaultValue=75, parameterName=Leg_R_Radius_stage1}
  n16 = NamedRerouteUsage()
  n17 = LinearInterpolate(
   A=LinearInterpolate(
    A=LinearInterpolate(B=@n15, Alpha=@n16)
    B=ScalarParameter() {defaultValue=102.771, parameterName=Leg_R_Radius_stage2}
    Alpha=NamedRerouteUsage()
   )
   B=ScalarParameter() {defaultValue=140, parameterName=Leg_R_Radius_stage3}
   Alpha=NamedRerouteUsage()
  )
  n18 = ScalarParameter() {defaultValue=0.97, parameterName=Leg_R_Hardness}
  n19 = TextureSample(UVs=Multiply(A=TextureCoordinate(), B=ScalarParameter() {defaultValue=2.8593, parameterName=Leg_L_tiling_stage2})) {samplerSource=SSM_Wrap_WorldGroupSettings, samplerType=SAMPLERTYPE_LinearColor, texture=&T_CloudNoise_01}
  n20 = LinearInterpolate(Alpha=@n19.B) {constA=-1}
  n21 = LinearInterpolate(
   A=ScalarParameter() {defaultValue=3.6713, parameterName=Leg_L_Strength_stage1}
   B=ScalarParameter() {defaultValue=29.3712, parameterName=Leg_L_Strength_stage2}
   Alpha=NamedRerouteUsage()
  )
  n22 = NamedRerouteUsage()
  n23 = Constant4Vector() {constant=<rgba -84.6653 0 -54.1491 1>}
  n24 = LinearInterpolate(B=ScalarParameter() {defaultValue=76, parameterName=Leg_L_Radius_stage1}, Alpha=NamedRerouteUsage())
  n25 = ScalarParameter() {defaultValue=111.5165, parameterName=Leg_L_Radius_stage2}
  n26 = NamedRerouteUsage()
  n27 = SphereMask(
   A=Add(A=Multiply(A=@n20, B=@n21), B=Add(A=@n22, B=@n23.RGB))
   B=Constant()
   Radius=LinearInterpolate(
    A=LinearInterpolate(A=@n24, B=@n25, Alpha=@n26)
    B=ScalarParameter() {defaultValue=150, parameterName=Leg_L_Radius_stage3}
    Alpha=NamedRerouteUsage()
   )
   Hardness=ScalarParameter() {defaultValue=0.97, parameterName=Leg_L_Hardness}
  )
  n28 = LocalPosition() {localOrigin=InstancePreSkinning}
  n29 = Constant() {r=9.6215}
  n30 = Constant() {r=4.5}
  n31 = TextureSample(UVs=Multiply(A=TextureCoordinate(), B=AppendVector(A=@n29, B=@n30))) {samplerSource=SSM_Wrap_WorldGroupSettings, samplerType=SAMPLERTYPE_Masks, texture=/Game/Global/Textures/Default/Noise/T_Noise_Cloud_Mask}
  n32 = Constant() {r=22.6961}
  n33 = LinearInterpolate(B=ScalarParameter() {defaultValue=55.1231, parameterName=Foot_Radius}, Alpha=NamedRerouteUsage())
  n34 = ScalarParameter() {defaultValue=129.0307, parameterName=Foot_Radius_stage2}
  n35 = NamedRerouteUsage()
  n36 = SphereMask(
   A=Add(A=VertexInterpolator(VS=@n28.XYZ).PS, B=Multiply(A=@n31.RGB, B=@n32))
   B=Constant()
   Radius=LinearInterpolate(
    A=LinearInterpolate(A=@n33, B=@n34, Alpha=@n35)
    B=ScalarParameter() {defaultValue=173.5603, parameterName=Foot_Radius_stage3}
    Alpha=NamedRerouteUsage()
   )
   Hardness=ScalarParameter() {defaultValue=0.97, parameterName=Foot_Hardness}
  )
  n37 = Max(
   A=SphereMask(A=Add(A=@n8, B=@n9), B=Constant(), Radius=LinearInterpolate(B=@n10, Alpha=@n11), Hardness=Constant() {r=0.97})
   B=Max(A=SphereMask(A=@n13, B=@n14, Radius=@n17, Hardness=@n18), B=Max(A=@n27, B=@n36))
  )
  n38 = NamedRerouteDeclaration(None=Max(A=Max(A=@n6, B=@n37), B=NamedRerouteUsage())) {name=OutMask, nodeColor=<rgba 0 0.5547 1 1>}
  FunctionOutput(None=@n38) {outputName=OutMask}
  n39 = Constant() {r=0.6143}
  n40 = Constant()
  n41 = Multiply(A=Multiply(A=SmoothStep(Min=@n39, Max=@n40, Value=@n38), B=@n38), B=ScalarParameter() {defaultValue=5, parameterName=GlowEdgeIntensity}) {constB=0.5}
  n42 = ScalarParameter() {group=SheerWill_BlendMask, parameterName=SheerWillEmoteAnimation}
  n43 = MaterialFunctionCall() {materialFunction=&Pi}
  n44 = NamedRerouteDeclaration(None=Sine(None=Multiply(A=@n42, B=@n43.Result)) {period=6.283}) {name=Sine_Stage2_3, nodeColor=<rgba 0 1 0.6016 1>}
  n45 = LinearInterpolate(A=ScalarParameter() {defaultValue=0.015, parameterName=GlowEdgeIntensityMIn}, Alpha=Max(A=Saturate(None=@n44))) {constA=0.1}
  FunctionOutput(None=Multiply(A=Multiply(A=@n41, B=@n45), B=VectorParameter() {parameterName=EmoteEdgeColor, defaultValue=<rgba 2.081 0.3564 3 1>}.RGB)) {outputName=EmoteEdgeColor}
  LinearInterpolate(
   B=Multiply(A=@n38, B=ScalarParameter() {defaultValue=1, parameterName=Global Mask Intensity})
   Alpha=ScalarParameter() {defaultValue=1, parameterName=UseReactive}
  ) {constA=1}
  n46 = Multiply(A=Multiply(
    A=ScalarParameter() {group=SheerWill_BlendMask, parameterName=SheerWillEmoteAnimation}
    B=ScalarParameter() {defaultValue=1, group=SheerWill_BlendMask, parameterName=EmoteAnimation_SpeedOffset}
   )) {constB=4}
  n47 = Subtract(A=@n46)
  n48 = ScalarParameter() {defaultValue=1, parameterName=SpeedOffset}
  NamedRerouteDeclaration(None=Saturate(None=Multiply(A=@n47, B=@n48))) {name=Animation_Stage1_2, nodeColor=<rgba 0.7344 0 1 1>}
  NamedRerouteDeclaration(None=Saturate(None=Multiply(A=@n46, B=@n48))) {name=Animation_Stage0_1, nodeColor=<rgba 0.7344 0 1 1>}
  NamedRerouteDeclaration(None=ScalarParameter() {group=SheerWill_BlendMask, parameterName=NoHelmet}) {name=NOHelmet, nodeColor=<rgba 0 0.4844 1 1>}
  n49 = Subtract(A=@n46) {constB=2}
  NamedRerouteDeclaration(None=Saturate(None=Multiply(A=@n49, B=@n48))) {name=Animation_Stage2_3, nodeColor=<rgba 0.7344 0 1 1>}
  n50 = Subtract(A=@n46) {constB=4}
  NamedRerouteDeclaration(None=Saturate(None=Multiply(A=@n50, B=@n48))) {name=Animation_Stage3_4, nodeColor=<rgba 0.7344 0 1 1>}
  n51 = ScalarParameter() {defaultValue=2.5, parameterName=Head_R_tiling_stage1}
  n52 = ScalarParameter() {defaultValue=2.5, parameterName=Head_R_tiling_stage2}
  n53 = NamedRerouteUsage()
  n54 = TextureSample(UVs=Multiply(A=TextureCoordinate(), B=LinearInterpolate(A=@n51, B=@n52, Alpha=@n53))) {samplerType=SAMPLERTYPE_LinearColor, texture=&T_CloudNoise_01}
  n55 = ScalarParameter() {defaultValue=3.6713, parameterName=Head_R_Strength_stage1}
  n56 = ScalarParameter() {defaultValue=0.85, parameterName=Head_R_Strength_stage1_NoHelmet}
  n57 = NamedRerouteUsage()
  n58 = VectorParameter() {parameterName=Head_R_LocalPosition, defaultValue=<rgba 91 0 -186.5374 1>}
  n59 = VectorParameter() {parameterName=Head_R_LocalPosition_NoHelmet, defaultValue=<rgba -84 0 -190.7936 1>}
  n60 = NamedRerouteUsage()
  n61 = Add(
   A=Multiply(A=LinearInterpolate(Alpha=@n54.B) {constA=-1}, B=LinearInterpolate(A=@n55, B=@n56, Alpha=@n57))
   B=Add(A=NamedRerouteUsage(), B=LinearInterpolate(A=@n58.RGB, B=@n59.RGB, Alpha=@n60))
  )
  n62 = Constant()
  n63 = LinearInterpolate(
   A=ScalarParameter() {defaultValue=81.4901, parameterName=Head_R_Radius_stage1}
   B=ScalarParameter() {defaultValue=87.0905, parameterName=Head_R_Radius_stage1_NoHelmet}
   Alpha=NamedRerouteUsage()
  )
  n64 = NamedRerouteUsage()
  n65 = ScalarParameter() {defaultValue=82, parameterName=Head_R_Radius_stage2}
  n66 = ScalarParameter() {defaultValue=89, parameterName=Head_R_Radius_stage2_NoHelmet}
  n67 = NamedRerouteUsage()
  n68 = LinearInterpolate(
   A=LinearInterpolate(A=LinearInterpolate(B=@n63, Alpha=@n64), B=LinearInterpolate(A=@n65, B=@n66, Alpha=@n67), Alpha=NamedRerouteUsage())
   B=LinearInterpolate(
    A=ScalarParameter() {defaultValue=120, parameterName=Head_R_Radius_stage3}
    B=ScalarParameter() {defaultValue=105, parameterName=Head_R_Radius_NoHelmet}
    Alpha=NamedRerouteUsage()
   )
   Alpha=NamedRerouteUsage()
  )
  n69 = ScalarParameter() {defaultValue=0.99, parameterName=Head_R_Hardness}
  n70 = TextureSample(UVs=Multiply(A=TextureCoordinate(), B=ScalarParameter() {defaultValue=2.5, parameterName=Head_L_tiling_stage1})) {samplerType=SAMPLERTYPE_LinearColor, texture=&T_CloudNoise_01}
  n71 = Add(
   A=Multiply(A=LinearInterpolate(Alpha=@n70.B) {constA=-1}, B=ScalarParameter() {defaultValue=2.2, parameterName=Head_L_Strength_stage1})
   B=Add(A=NamedRerouteUsage(), B=VectorParameter() {parameterName=Head_L_LocalPosition, defaultValue=<rgba -56.2564 0 -237.6378 1>}.RGB)
  )
  n72 = Constant()
  n73 = ScalarParameter() {defaultValue=85.6025, parameterName=Head_L_Radius_stage1}
  n74 = NamedRerouteUsage()
  n75 = LinearInterpolate(
   A=LinearInterpolate(
    A=LinearInterpolate(B=@n73, Alpha=@n74)
    B=ScalarParameter() {defaultValue=89.3057, parameterName=Head_L_Radius_stage2}
    Alpha=NamedRerouteUsage()
   )
   B=ScalarParameter() {defaultValue=120, parameterName=Head_L_Radius_stage3}
   Alpha=NamedRerouteUsage()
  )
  n76 = ScalarParameter() {defaultValue=0.99, parameterName=Head_L_Hardness}
  NamedRerouteDeclaration(None=Max(A=SphereMask(A=@n61, B=@n62, Radius=@n68, Hardness=@n69), B=SphereMask(A=@n71, B=@n72, Radius=@n75, Hardness=@n76))) {name=HeadMask, nodeColor=<rgba 1 0 0.7969 1>}
  NamedRerouteDeclaration(None=VertexInterpolator(VS=LocalPosition() {localOrigin=InstancePreSkinning}.XYZ).PS) {name=PreSkinnedLocalPosition, nodeColor=<rgba 0 1 0.0625 1>}
 ]}
