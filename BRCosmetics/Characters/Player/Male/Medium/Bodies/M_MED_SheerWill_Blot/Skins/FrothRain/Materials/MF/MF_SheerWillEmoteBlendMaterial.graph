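# Material function for the SheerWill Blot FrothRain skin that blends incoming material attributes toward an emote look.
# The emote layer overrides the input with black base color, low specular, full roughness and an emissive combining a reroute with the emote mask's edge color.
# A fresnel mask tinted by the Emote_FresnelColor parameter and dimmed forms the OutEmissiveEmote glow.
# MF_SheerWill_EmoteMask's output mask drives the blend between original and emote attributes.
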MaterialFunction /BRCosmetics/Characters/Player/Male/Medium/Bodies/M_MED_SheerWill_Blot/Skins/FrothRain/Materials/MF/MF_SheerWillEmoteBlendMaterial
graph {nodes [
  n1 = FunctionInput() {bUsePreviewValueAsDefault=true, inputType=FunctionInput_MaterialAttributes}
  n2 = MaterialFunctionCall() {materialFunction=&/MF_SheerWill_EmoteMask}
  n3 = SetMaterialAttributes!(
   MaterialAttributes=@n1
   BaseColor=Constant()
   EmissiveColor=Add(A=NamedRerouteUsage(), B=@n2.EmoteEdgeColor)
   Metallic
   Specular=Constant() {r=0.1}
   Roughness=Constant() {r=1}
   OpacityMask
  ) {attributeSetTypes [
    #1
    #2
    #3
    #4
    #5
    #6
   ]}
  FunctionOutput(None=BlendMaterialAttributes(A=@n1, B=@n3, Alpha=@n2.OutMask))
  n4 = ComponentMask(None=MaterialFunctionCall() {materialFunction=/BRCosmetics/Characters/Player/Male/Medium/Bodies/M_MED_SheerWill_Phase/Material/MF_SheerWill_FresnelMask}.SheerWill_FresnelMask) {r=true}
  n5 = VectorParameter() {parameterName=Emote_FresnelColor, defaultValue=<rgba 0.6827 0.0014 1 1>}
  NamedRerouteDeclaration(None=Multiply(A=Multiply(A=@n4, B=@n5.RGB)) {constB=0.4}) {name=OutEmissiveEmote, nodeColor=<rgba 0.6641 0 1 1>}
 ]}
