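# Material function for the FrothRain skin of SheerWill Blot that builds a staged, noise-edged face accessory mask, outputting OutMask and four saturated stage-blend values.
# Left and right sphere masks on pre-skinned local position, distorted by cloud noise, combine by Max into HeadMask.
# Sphere radii interpolate across stages, with NoHelmet variants for the right sphere.
# Stage parameters are scaled by SpeedOffset; UseReactive blends OutMask from 1 to an intensity-scaled mask.
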MaterialFunction /BRCosmetics/Characters/Player/Male/Medium/Bodies/M_MED_SheerWill_Blot/Skins/FrothRain/Materials/MF/MF_SheerWill_FaceAccMask
graph {nodes [
  n1 = Multiply(A=NamedRerouteUsage(), B=ScalarParameter() {defaultValue=1, parameterName=Global Mask Intensity})
  n2 = ScalarParameter() {defaultValue=1, parameterName=UseReactive}
  FunctionOutput(None=NamedRerouteDeclaration(None=LinearInterpolate(B=@n1, Alpha=@n2) {constA=1}) {name=OutMask, nodeColor=<rgba 0 1 0.5781 1>}) {outputName=OutMask}
  n3 = ScalarParameter() {defaultValue=1, parameterName=SpeedOffset}
  n4 = Multiply(A=ScalarParameter() {group=SheerWill_BlendMask, parameterName=Stage2_3}, B=LinearInterpolate(A=@n3, Alpha=NamedRerouteUsage()))
  FunctionOutput(None=NamedRerouteDeclaration(None=Saturate(None=@n4)) {name=Animation_Stage2_3, nodeColor=<rgba 0.7344 0 1 1>}) {outputName=Stage2_3}
  n5 = Multiply(A=ScalarParameter() {group=SheerWill_BlendMask, parameterName=Stage3_4}, B=@n3)
  FunctionOutput(None=NamedRerouteDeclaration(None=Saturate(None=@n5)) {name=Animation_Stage3_4, nodeColor=<rgba 0.7344 0 1 1>}) {outputName=Stage3_4}
  n6 = Multiply(A=ScalarParameter() {group=SheerWill_BlendMask, parameterName=Stage0_1}, B=@n3)
  FunctionOutput(None=NamedRerouteDeclaration(None=Saturate(None=@n6)) {name=Animation_Stage0_1, nodeColor=<rgba 0.7344 0 1 1>}) {outputName=Stage0_1}
  n7 = Multiply(A=ScalarParameter() {group=SheerWill_BlendMask, parameterName=Stage1_2}, B=@n3)
  FunctionOutput(None=NamedRerouteDeclaration(None=Saturate(None=@n7)) {name=Animation_Stage1_2, nodeColor=<rgba 0.7344 0 1 1>}) {outputName=Stage1_2}
  n8 = TextureCoordinate()
  n9 = ScalarParameter() {defaultValue=2.5, parameterName=Head_L_tiling_stage1}
  n10 = LinearInterpolate(Alpha=TextureSample(UVs=Multiply(A=@n8, B=@n9)) {samplerType=SAMPLERTYPE_LinearColor, texture=/Game/Global/Textures/Default/Noise/T_CloudNoise_01}.B) {constA=-1}
  n11 = ScalarParameter() {defaultValue=2.2, parameterName=Head_L_Strength_stage1}
  n12 = NamedRerouteUsage()
  n13 = VectorParameter() {parameterName=Head_L_LocalPosition, defaultValue=<rgba -56.2564 0 -237.6378 1>}
  n14 = LinearInterpolate(B=ScalarParameter() {defaultValue=85.6025, parameterName=Head_L_Radius_stage1}, Alpha=NamedRerouteUsage())
  n15 = ScalarParameter() {defaultValue=89.3057, parameterName=Head_L_Radius_stage2}
  n16 = NamedRerouteUsage()
  n17 = SphereMask(
   A=Add(A=Multiply(A=@n10, B=@n11), B=Add(A=@n12, B=@n13.RGB))
   B=Constant()
   Radius=LinearInterpolate(
    A=LinearInterpolate(A=@n14, B=@n15, Alpha=@n16)
    B=ScalarParameter() {defaultValue=120, parameterName=Head_L_Radius_stage3}
    Alpha=NamedRerouteUsage()
   )
   Hardness=ScalarParameter() {defaultValue=0.99, parameterName=Head_L_Hardness}
  )
  LinearInterpolate(A=@n17, Alpha=NamedRerouteUsage()) {constB=0}
  n18 = ScalarParameter() {defaultValue=2.5, parameterName=Head_R_tiling_stage1}
  n19 = ScalarParameter() {defaultValue=2.5, parameterName=Head_R_tiling_stage2}
  n20 = NamedRerouteUsage()
  n21 = TextureSample(UVs=Multiply(A=TextureCoordinate(), B=LinearInterpolate(A=@n18, B=@n19, Alpha=@n20))) {samplerType=SAMPLERTYPE_LinearColor, texture=&T_CloudNoise_01}
  n22 = ScalarParameter() {defaultValue=3.6713, parameterName=Head_R_Strength_stage1}
  n23 = ScalarParameter() {defaultValue=0.85, parameterName=Head_R_Strength_stage1_NoHelmet}
  n24 = NamedRerouteUsage()
  n25 = VertexInterpolator(VS=LocalPosition() {localOrigin=InstancePreSkinning}.XYZ)
  n26 = VectorParameter() {parameterName=Head_R_LocalPosition, defaultValue=<rgba 91 0 -186.5374 1>}
  n27 = VectorParameter() {parameterName=Head_R_LocalPosition_NoHelmet, defaultValue=<rgba -84 0 -190.7936 1>}
  n28 = NamedRerouteUsage()
  n29 = Add(
   A=Multiply(A=LinearInterpolate(Alpha=@n21.B) {constA=-1}, B=LinearInterpolate(A=@n22, B=@n23, Alpha=@n24))
   B=Add(
    A=NamedRerouteDeclaration(None=@n25.PS) {name=PreSkinnedLocalPosition, nodeColor=<rgba 0 1 0.0625 1>}
    B=LinearInterpolate(A=@n26.RGB, B=@n27.RGB, Alpha=@n28)
   )
  )
  n30 = Constant()
  n31 = LinearInterpolate(
   A=ScalarParameter() {defaultValue=81.4901, parameterName=Head_R_Radius_stage1}
   B=ScalarParameter() {defaultValue=87.0905, parameterName=Head_R_Radius_stage1_NoHelmet}
   Alpha=NamedRerouteUsage()
  )
  n32 = NamedRerouteUsage()
  n33 = ScalarParameter() {defaultValue=82, parameterName=Head_R_Radius_stage2}
  n34 = ScalarParameter() {defaultValue=89, parameterName=Head_R_Radius_stage2_NoHelmet}
  n35 = NamedRerouteUsage()
  n36 = LinearInterpolate(
   A=LinearInterpolate(A=LinearInterpolate(B=@n31, Alpha=@n32), B=LinearInterpolate(A=@n33, B=@n34, Alpha=@n35), Alpha=NamedRerouteUsage())
   B=LinearInterpolate(
    A=ScalarParameter() {defaultValue=120, parameterName=Head_R_Radius_stage3}
    B=ScalarParameter() {defaultValue=105, parameterName=Head_R_Radius_NoHelmet}
    Alpha=NamedRerouteUsage()
   )
   Alpha=NamedRerouteUsage()
  )
  n37 = ScalarParameter() {defaultValue=0.99, parameterName=Head_R_Hardness}
  NamedRerouteDeclaration(None=Max(A=SphereMask(A=@n29, B=@n30, Radius=@n36, Hardness=@n37), B=@n17)) {name=HeadMask, nodeColor=<rgba 1 0 0.7969 1>}
  NamedRerouteDeclaration(None=ScalarParameter() {group=SheerWill_BlendMask, parameterName=NoHelmet}) {name=NOHelmet, nodeColor=<rgba 0 0.4844 1 1>}
  ScalarParameter() {defaultValue=1.5, parameterName=SpeedOffset3_4}
 ]}
