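# Material function for the SheerWill Blot FrothRain character skin that restyles incoming material attributes as painted, anisotropic hair.
# Strand mask and normal textures, Diffuse2 and hue shifts feed color, flattened blended normals, noisy roughness, specular, AO and scraggled tangents, with switches to keep base values.
# Several parameters lerp between paired variants.
# Fresnel brightens base color and fakes light transmission, scaled by a collection emissive multiplier that lets capture dim the glow.
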MaterialFunction /BRCosmetics/Characters/Player/Male/Medium/Bodies/M_MED_SheerWill_Blot/Skins/FrothRain/Materials/MF/MF_SheerWill_Hair
graph {nodes [
  n1 = VectorParameter() {group=FN_Hair, parameterName=Hair_Color_Variation, sortPriority=1, defaultValue=<rgba 1 0.0335 0 1>}
  n2 = NamedRerouteUsage()
  n3 = Multiply(A=NamedRerouteUsage(), B=Constant() {r=1})
  n4 = Multiply(A=NamedRerouteUsage(), B=NamedRerouteUsage())
  n5 = ComponentMask(None=MaterialFunctionCall(
    Hue Shift Percentage=NamedRerouteUsage()
    Texture=MaterialFunctionCall(Hue Shift Percentage=@n3, Texture=@n4) {materialFunction=&HueShift}.Result
   ) {materialFunction=&HueShift}.Result) {r=true}
  n6 = LinearInterpolate(A=NamedRerouteUsage(), B=Saturate(None=Add(A=@n1.RGB, B=@n2)), Alpha=Power(Base=Multiply(A=@n5) {constB=1.3}))
  FunctionOutput(None=@n6) {outputName=OutPut paint hair texture}
  n7 = NamedRerouteUsage()
  n8 = ScalarParameter() {group=FN_Hair, parameterName=EmissiveStrenghParam}
  n9 = ScalarParameter() {defaultValue=0.6, group=FN_Hair_Advanced, parameterName=AmbientOcclusion_Black}
  n10 = ScalarParameter() {defaultValue=0.8, group=FN_Hair_Advanced, parameterName=AmbientOcclusion_Black2}
  n11 = NamedRerouteUsage()
  n12 = GetMaterialAttributes(None=FunctionInput() {bUsePreviewValueAsDefault=true, inputType=FunctionInput_MaterialAttributes}) {attributeGetTypes [
    #1
    #2
    #3
    #4
    #5
    #6
    #7
    #8
   ]}
  n13 = NamedRerouteUsage()
  n14 = Subtract(
   A=MaterialFunctionCall(X=@n13, Y=Constant(), Z=@n13) {materialFunction=&MakeFloat3}.Result
   B=Constant3Vector() {constant=<rgba 0.5 0.5 0 0>}
  )
  n15 = ScalarParameter() {defaultValue=0.25, group=FN_Hair_Advanced, parameterName=Scraggle}
  n16 = SetMaterialAttributes!(
   MaterialAttributes=@n12.MaterialAttributes
   Roughness=NamedRerouteUsage()
   Specular=NamedRerouteUsage()
   Metallic=NamedRerouteUsage()
   Anisotropy=NamedRerouteUsage()
   BaseColor=NamedRerouteUsage()
   Normal=NamedRerouteUsage()
   ShadingModel=ShadingModel()
   EmissiveColor=Add(A=NamedRerouteUsage(), B=Multiply(A=@n7, B=@n8))
   Tangent=Add(A=NamedRerouteUsage(), B=Multiply(A=@n14, B=@n15))
   AmbientOcclusion=LinearInterpolate(A=LinearInterpolate(A=@n9, B=@n10, Alpha=@n11), Alpha=NamedRerouteUsage()) {constA=0.6}
  ) {attributeSetTypes [
    #8
    #7
    #6
    #4
    #1
    #2
    #9
    #5
    #3
    #10
   ]}
  FunctionOutput(None=@n16)
  FunctionOutput(None=NamedRerouteDeclaration(None=@n12.Anisotropy) {name=InPut_AnisotropyWeight, nodeColor=<rgba 1 0 0.8437 1>}) {outputName=AnisotropyWeight}
  n17 = TextureSampleParameter2D() {
   group=00_Texture Inputs
   parameterName=Diffuse2
   samplerSource=SSM_Wrap_WorldGroupSettings
   sortPriority=0
   texture=/Game/Global/Textures/Default/Blanks/T_MidgreysRGB
  }
  n18 = NamedRerouteUsage()
  FunctionOutput(None=NamedRerouteDeclaration(None=LinearInterpolate(A=@n12.BaseColor, B=@n17.RGB, Alpha=@n18)) {name=InPutBaseColor, nodeColor=<rgba 0.2422 1 0 1>}) {outputName=OutPutBasecolor}
  n19 = TextureSampleParameter2D() {
   desc="R:Hair mask G; Hairstrands"
   group=FN_Hair
   parameterName=Hair Mask
   samplerSource=SSM_Wrap_WorldGroupSettings
   samplerType=SAMPLERTYPE_Masks
   texture=&T_BlackMask
  }
  FunctionOutput(None=@n19.R) {outputName=HairMask}
  n20 = NamedRerouteUsage()
  n21 = LinearInterpolate(
   A=ScalarParameter() {defaultValue=0.5, group=FN_Hair, parameterName=Hair_Anisotropy_Min}
   B=ScalarParameter() {defaultValue=0.75, group=FN_Hair, parameterName=Hair_Anisotropy_Max}
   Alpha=Saturate(None=NamedRerouteUsage())
  )
  NamedRerouteDeclaration(None=LinearInterpolate(A=@n21, B=Multiply(A=@n21, B=@n20), Alpha=Constant() {desc="Default 1, use Weight map 1", r=1})) {name=Anisotropy, nodeColor=<rgba 0.0781 0 1 1>}
  NamedRerouteDeclaration(None=@n12.Metallic) {name=IN_Metallic, nodeColor=<rgba 1 0.5391 0 1>}
  NamedRerouteDeclaration(None=Power(Base=NamedRerouteUsage(), Exp=ScalarParameter() {defaultValue=1, group=FN_Hair, parameterName=Specular _POWER})) {name=Out specular, nodeColor=<rgba 0 0.2266 1 1>}
  n22 = NamedRerouteUsage()
  NamedRerouteDeclaration(None=Power(Base=OneMinus(None=@n22), Exp=ScalarParameter() {defaultValue=1, group=FN_Hair, parameterName=Roughness_power, sortPriority=5}) {constExponent=4}) {name=OutPutRoughness, nodeColor=<rgba 1 0 0.6797 1>}
  NamedRerouteDeclaration(None=MaterialFunctionCall(In=@n19.G, Contrast=ScalarParameter() {group=FN_Hair, parameterName=Paint_Hair_Contrast}) {materialFunction=&CheapContrast}.Result) {name=InPut_Hair_Texture, nodeColor=<rgba 0.1016 0 1 1>}
  n23 = CollectionParameter() {
   collection=&FortniteMaterialParameters
   desc="CAPTURE EMISSIVE-GLOW GATE  —  FORT-1106795\r\n\r\nWHY: Capture's controlled/dark lighting makes this glow blow out on dark hair && fur.This lets them dim/kill it for a shot without touching the authored look.\r\n\r\nDRIVEN BY: console cvar  Fort.Cosmetics.EmissiveMultiplier"
   parameterName=CosmeticsEmissiveMultiplier
  }
  n24 = ScalarParameter() {defaultValue=1, group=FN_Hair, parameterName=Emissive_Brightness, sortPriority=1}
  n25 = NamedRerouteUsage()
  n26 = ScalarParameter() {defaultValue=0.5, group=FN_Hair, parameterName=BaseColor_Fresnel_Brightness, sortPriority=1}
  n27 = ScalarParameter() {defaultValue=0.6, group=FN_Hair, parameterName=BaseColor_Fresnel_Brightness2, sortPriority=1}
  n28 = NamedRerouteUsage()
  n29 = ScalarParameter() {defaultValue=3.5, group=FN_Hair_Advanced, parameterName=Basecolor_Fresnel_Exponent, sortPriority=1}
  n30 = ScalarParameter() {defaultValue=3.5, group=FN_Hair_Advanced, parameterName=Basecolor_Fresnel_Exponent2, sortPriority=1}
  n31 = NamedRerouteUsage()
  n32 = NamedRerouteDeclaration(None=MaterialFunctionCall(
    Brightness=LinearInterpolate(A=@n26, B=@n27, Alpha=@n28)
    Exponent=LinearInterpolate(A=@n29, B=@n30, Alpha=@n31)
    Vertex / Pixel Normal=PixelNormalWS()
   ) {materialFunction=&MF_Fresnel_FX}.Fresnel Inverted Double Sided) {name=fresnel, nodeColor=<rgba 0 1 0.2969 1>}
  n33 = Multiply(A=@n25, B=@n32)
  NamedRerouteDeclaration(None=Multiply(A=Multiply(A=@n33, B=@n24), B=@n23)) {name=simulate light transmission, nodeColor=<rgba 0 1 0.8594 1>}
  NamedRerouteDeclaration(None=@n12.Roughness) {name=IN_Roughness, nodeColor=<rgba 1 0.5391 0 1>}
  n34 = MaterialFunctionCall(
   Normal=NamedRerouteUsage()
   Flatness=ScalarParameter() {defaultValue=0.6, group=FN_Hair, parameterName=Hair_Mesh_Normal_Flatness, sortPriority=2}
  ) {materialFunction=&FlattenNormal}
  n35 = TextureSampleParameter2D() {
   group=FN_Hair
   parameterName=Strands Normal
   samplerSource=SSM_Wrap_WorldGroupSettings
   samplerType=SAMPLERTYPE_Normal
   texture=/Game/Characters/Player/Female/Medium/Bodies/F_MED_RebelClaw_Aviator/Skins/Cobalt/Textures/T_Blank_Normal
  }
  n36 = MaterialFunctionCall(Normal=@n35.RGB, Flatness=ScalarParameter() {defaultValue=0.6, group=FN_Hair, parameterName=Paint_Hair_Normal_Flatness, sortPriority=2}) {materialFunction=&FlattenNormal}
  NamedRerouteDeclaration(None=Normalize(VectorInput=MaterialFunctionCall(BaseNormal=@n34.Result, AdditionalNormal=@n36.Result) {materialFunction=&BlendAngleCorrectedNormals}.Result)) {name=Normal, nodeColor=<rgba 0.5 1 0 1>}
  NamedRerouteDeclaration(None=@n12.Specular) {name=IN_Spec, nodeColor=<rgba 1 0.5391 0 1>}
  NamedRerouteDeclaration(None=@n12.Tangent) {name=InPutTangentMap, nodeColor=<rgba 0.875 1 0 1>}
  n37 = Multiply(A=@n25, B=@n32)
  n38 = ScalarParameter() {defaultValue=8, group=FN_Hair, parameterName=Fresnel_Brightness_Multiple, sortPriority=1}
  n39 = Saturate(None=Add(A=Multiply(A=@n37, B=@n38), B=@n25))
  n40 = NamedRerouteUsage()
  n41 = ScalarParameter() {group=FN_Hair_Advanced, parameterName=Gmap_intensity, sortPriority=1}
  n42 = ScalarParameter() {defaultValue=0.5, group=FN_Hair_Advanced, parameterName=Gmap_intensity2, sortPriority=1}
  n43 = NamedRerouteUsage()
  NamedRerouteDeclaration(None=LinearInterpolate(A=@n39, B=Multiply(A=@n39, B=@n40), Alpha=LinearInterpolate(A=@n41, B=@n42, Alpha=@n43))) {name=basecolor with fresnel, nodeColor=<rgba 0.8984 0 1 1>}
  n44 = MaterialFunctionCall(Float3=FunctionInput() {bUsePreviewValueAsDefault=true, inputName=Mask Input, sortPriority=1}) {desc="R = AO\r\nG = GMap\r\nB = Skin", materialFunction=&BreakOutFloat3Components}
  NamedRerouteDeclaration(None=@n44.G) {name=InPutGmap, nodeColor=<rgba 0 1 0.1328 1>}
  NamedRerouteDeclaration(None=FunctionInput(Preview=Constant()) {bUsePreviewValueAsDefault=true, inputName=BlendMask, inputType=FunctionInput_Scalar}) {name=BlendMask, nodeColor=<rgba 0.7344 0 1 1>}
  NamedRerouteDeclaration(None=StaticSwitchParameter(True=NamedRerouteUsage(), False=ScalarParameter() {defaultValue=0.05, group=FN_Hair, parameterName=Hair_Metallic}) {group=FN_Hair, parameterName=UseBaseMetallic, sortPriority=102}) {name=Metallic, nodeColor=<rgba 0 0.7422 1 1>}
  n45 = LinearInterpolate(
   A=ScalarParameter() {defaultValue=0.3, group=FN_Hair, parameterName=Hair_Specular_MIN}
   B=ScalarParameter() {defaultValue=0.5, group=FN_Hair, parameterName=Hair_Specular_MAX}
   Alpha=NamedRerouteUsage()
  )
  NamedRerouteDeclaration(None=StaticSwitchParameter(True=NamedRerouteUsage(), False=Saturate(None=@n45)) {group=FN_Hair, parameterName=UseBaseSpecular, sortPriority=101}) {name=Specular, nodeColor=<rgba 1 0.9375 0 1>}
  NamedRerouteDeclaration(None=@n44.R) {name=InPut Ambient Occlusion, nodeColor=<rgba 0.9453 1 0 1>}
  n46 = VectorParameter() {group=FN_Hair, parameterName=Paint_Hair_Color_Darkness, sortPriority=1, defaultValue=<rgba 0.5 0.5 0.5 1>}
  n47 = VectorParameter() {group=FN_Hair, parameterName=Paint_Hair_Color_Darkness2, sortPriority=1, defaultValue=<rgba 0.5 0.5 0.5 1>}
  n48 = LinearInterpolate(A=@n46.RGB, B=@n47.RGB, Alpha=NamedRerouteUsage())
  n49 = VectorParameter() {group=FN_Hair, parameterName=Paint_Hair_Color_Brightness, sortPriority=1, defaultValue=<rgba 1 1 1 1>}
  NamedRerouteDeclaration(None=LinearInterpolate(A=Multiply(A=@n6, B=@n48) {constB=0.1}, B=Multiply(A=@n6, B=@n49.RGB) {constB=1.5}, Alpha=NamedRerouteUsage())) {name=OutPutBaseColor, nodeColor=<rgba 0.2891 1 0 1>}
  NamedRerouteDeclaration(None=@n12.Normal) {name=InPutNM, nodeColor=<rgba 1 0.5625 0 1>}
  NamedRerouteDeclaration(None=@n12.EmissiveColor) {name=EmissiveColor, nodeColor=<rgba 1 0 0.2578 1>}
  n50 = LinearInterpolate(
   A=ScalarParameter() {defaultValue=0.75, group=FN_Hair, parameterName=Roughness Min, sortPriority=5}
   B=ScalarParameter() {defaultValue=0.84, group=FN_Hair, parameterName=Roughness Max, sortPriority=5}
   Alpha=Saturate(None=NamedRerouteUsage())
  )
  n51 = TextureCoordinate()
  n52 = ScalarParameter() {defaultValue=90, group=FN_Hair_Advanced, parameterName=Roughness_Noise_Tiling, sortPriority=5}
  n53 = LinearInterpolate(
   A=ScalarParameter() {defaultValue=0.75, group=FN_Hair, parameterName=Hair_Noise_Roughness_Min, sortPriority=5}
   Alpha=TextureSample(UVs=Multiply(A=@n51, B=@n52)) {samplerSource=SSM_Wrap_WorldGroupSettings, texture=/Game/Global/Textures/Default/Noise/Noise}.R
  ) {constA=0.5}
  NamedRerouteDeclaration(None=StaticSwitchParameter(True=NamedRerouteUsage(), False=Multiply(A=@n50, B=@n53)) {group=FN_Hair, parameterName=UseBaseRoughness, sortPriority=100}) {name=Roughness, nodeColor=<rgba 1 0.4453 0 1>}
  NamedRerouteDeclaration(None=@n44.B) {name=InPutSkin, nodeColor=<rgba 0.6641 0 1 1>}
 ]}
