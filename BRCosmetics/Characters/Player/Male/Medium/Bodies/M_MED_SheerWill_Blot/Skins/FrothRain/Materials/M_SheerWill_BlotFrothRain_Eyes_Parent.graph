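# Parent eye material for the SheerWill Blot FrothRain character skin, layering a two-stage reactive look over the shared MF_eyes function.
# Stage1_2 blacks out base color inside a UV sphere mask; Stage2_3 fades base color to black and emissive to the boosted head emissive texture.
# With UseReactive at zero, the stage inputs default to full.
# Global Mask Intensity blends the effect against the plain eyes before MF_TechArtOverrides.
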Material /BRCosmetics/Characters/Player/Male/Medium/Bodies/M_MED_SheerWill_Blot/Skins/FrothRain/Materials/M_SheerWill_BlotFrothRain_Eyes_Parent
graph {
 properties {
  bUseMaterialAttributes=true
  bUsedWithMeshDeformer=true
  bUsedWithSkeletalMesh=true
  pixelDepthOffsetMode=PDOM_Legacy
  shadingModel=MSM_FromMaterialExpression
  usedShadingModels=Default Lit
 }
 nodes [
  n1 = MaterialFunctionCall() {materialFunction=/Game/Characters/Shared/FNEye/Materials/MF_eyes}
  n2 = GetMaterialAttributes(None=@n1.Result) {attributeGetTypes [#1, #2]}
  n3 = SphereMask(A=TextureCoordinate(), B=AppendVector(A=Constant() {r=0.9654}, B=Constant() {r=0.48}), Radius=Constant() {r=0.25})
  n4 = ScalarParameter() {group=SheerWill_BlendMask, parameterName=Stage1_2}
  n5 = ScalarParameter() {defaultValue=1, group=SheerWill_BlendMask, parameterName=UseReactive}
  n6 = Saturate(None=Multiply(A=LinearInterpolate(B=@n4, Alpha=@n5) {constA=1}, B=ScalarParameter() {defaultValue=1, parameterName=SpeedOffset}))
  n7 = Saturate(None=LinearInterpolate(
    B=ScalarParameter() {group=SheerWill_BlendMask, parameterName=Stage2_3}
    Alpha=ScalarParameter() {defaultValue=1, group=SheerWill_BlendMask, parameterName=UseReactive}
   ) {constA=1})
  n8 = Multiply(
   A=LinearInterpolate(A=@n2.Base Color, B=Constant(), Alpha=LinearInterpolate(B=@n3, Alpha=@n6))
   B=LinearInterpolate(Alpha=@n7) {constA=1, constB=0}
  )
  n9 = LinearInterpolate(
   A=@n2.Emissive Color
   B=Multiply(A=TextureSample() {texture=/BRCosmetics/Characters/Player/Male/Medium/Bodies/M_MED_SheerWill_Blot/Skins/FrothRain/Textures/T_SheerWill_BlotFrothRain_Head_E}.RGB) {constB=2}
   Alpha=@n7
  )
  MaterialFunctionCall(In=BlendMaterialAttributes(
    A=@n1.Result
    B=SetMaterialAttributes!(MaterialAttributes=@n2.MaterialAttributes, Base Color=@n8, Emissive Color=@n9) {attributeSetTypes [#1, #2]}
    Alpha=ScalarParameter() {defaultValue=1, parameterName=Global Mask Intensity}
   )) {materialFunction=&MF_TechArtOverrides}
 ]
}
