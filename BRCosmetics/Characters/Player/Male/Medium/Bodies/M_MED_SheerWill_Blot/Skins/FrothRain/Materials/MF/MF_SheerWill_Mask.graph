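# Material function that builds a staged, reactive body mask for the SheerWill Blot FrothRain character skin.
# Noise-distorted SphereMasks on pre-skinned local position cover both arms, both legs and the feet, merged with Max.
# Arm radii lerp up from zero, while leg and foot radii chain through three larger stages.
# UseReactive blends the mask from solid white, and four Stage parameters, scaled by SpeedOffset and saturated, are also output.
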MaterialFunction /BRCosmetics/Characters/Player/Male/Medium/Bodies/M_MED_SheerWill_Blot/Skins/FrothRain/Materials/MF/MF_SheerWill_Mask
graph {nodes [
  n1 = TextureSample(UVs=Multiply(A=TextureCoordinate(), B=Constant() {r=3.1009})) {samplerSource=SSM_Wrap_WorldGroupSettings, samplerType=SAMPLERTYPE_LinearColor, texture=/Game/Global/Textures/Default/Noise/T_CloudNoise_01}
  n2 = Add(
   A=Multiply(A=LinearInterpolate(Alpha=@n1.B) {constA=-1}, B=Constant() {r=7.235})
   B=Add(A=NamedRerouteUsage(), B=Constant4Vector() {constant=<rgba -38.8338 0 -128.0196 1>}.RGB)
  )
  n3 = Constant()
  n4 = LinearInterpolate(B=ScalarParameter() {defaultValue=18.2411, parameterName=Arm_L_Radius}, Alpha=NamedRerouteUsage())
  n5 = Constant() {r=0.98}
  n6 = TextureSample(UVs=Multiply(A=TextureCoordinate(), B=Constant() {r=3.1009})) {samplerSource=SSM_Wrap_WorldGroupSettings, samplerType=SAMPLERTYPE_LinearColor, texture=&T_CloudNoise_01}
  n7 = LinearInterpolate(Alpha=@n6.B) {constA=-1}
  n8 = Constant() {r=7.235}
  n9 = NamedRerouteUsage()
  n10 = Constant4Vector() {constant=<rgba 51.6587 0 -133.7355 1>}
  n11 = SphereMask(
   A=Add(A=Multiply(A=@n7, B=@n8), B=Add(A=@n9, B=@n10.RGB))
   B=Constant()
   Radius=LinearInterpolate(B=ScalarParameter() {defaultValue=22, parameterName=Arm_R_Radius}, Alpha=NamedRerouteUsage())
   Hardness=Constant() {r=0.97}
  )
  n12 = TextureSample(UVs=Multiply(A=TextureCoordinate(), B=Constant() {r=3.1009})) {samplerSource=SSM_Wrap_WorldGroupSettings, samplerType=SAMPLERTYPE_LinearColor, texture=&T_CloudNoise_01}
  n13 = Multiply(A=LinearInterpolate(Alpha=@n12.B) {constA=-1}, B=Constant() {r=3.6713})
  n14 = Add(A=NamedRerouteUsage(), B=Constant4Vector() {constant=<rgba 89.2488 0 -54 1>}.RGB)
  n15 = LinearInterpolate(
   A=LinearInterpolate(B=ScalarParameter() {defaultValue=75, parameterName=Leg_R_Radius_stage1}, Alpha=NamedRerouteUsage())
   B=ScalarParameter() {defaultValue=102.771, parameterName=Leg_R_Radius_stage2}
   Alpha=NamedRerouteUsage()
  )
  n16 = ScalarParameter() {defaultValue=140, parameterName=Leg_R_Radius_stage3}
  n17 = NamedRerouteUsage()
  n18 = TextureSample(UVs=Multiply(A=TextureCoordinate(), B=ScalarParameter() {defaultValue=2.8593, parameterName=Leg_L_tiling_stage2})) {samplerSource=SSM_Wrap_WorldGroupSettings, samplerType=SAMPLERTYPE_LinearColor, texture=&T_CloudNoise_01}
  n19 = ScalarParameter() {defaultValue=3.6713, parameterName=Leg_L_Strength_stage1}
  n20 = ScalarParameter() {defaultValue=29.3712, parameterName=Leg_L_Strength_stage2}
  n21 = NamedRerouteUsage()
  n22 = Add(
   A=Multiply(A=LinearInterpolate(Alpha=@n18.B) {constA=-1}, B=LinearInterpolate(A=@n19, B=@n20, Alpha=@n21))
   B=Add(A=NamedRerouteUsage(), B=Constant4Vector() {constant=<rgba -84.6653 0 -54.1491 1>}.RGB)
  )
  n23 = Constant()
  n24 = ScalarParameter() {defaultValue=76, parameterName=Leg_L_Radius_stage1}
  n25 = NamedRerouteUsage()
  n26 = LinearInterpolate(
   A=LinearInterpolate(
    A=LinearInterpolate(B=@n24, Alpha=@n25)
    B=ScalarParameter() {defaultValue=111.5165, parameterName=Leg_L_Radius_stage2}
    Alpha=NamedRerouteUsage()
   )
   B=ScalarParameter() {defaultValue=150, parameterName=Leg_L_Radius_stage3}
   Alpha=NamedRerouteUsage()
  )
  n27 = ScalarParameter() {defaultValue=0.97, parameterName=Leg_L_Hardness}
  n28 = Constant() {r=9.6215}
  n29 = Constant() {r=4.5}
  n30 = TextureSample(UVs=Multiply(A=TextureCoordinate(), B=AppendVector(A=@n28, B=@n29))) {samplerSource=SSM_Wrap_WorldGroupSettings, samplerType=SAMPLERTYPE_Masks, texture=/Game/Global/Textures/Default/Noise/T_Noise_Cloud_Mask}
  n31 = Add(A=VertexInterpolator(VS=LocalPosition() {localOrigin=InstancePreSkinning}.XYZ).PS, B=Multiply(A=@n30.RGB, B=Constant() {r=22.6961}))
  n32 = Constant()
  n33 = ScalarParameter() {defaultValue=55.1231, parameterName=Foot_Radius}
  n34 = NamedRerouteUsage()
  n35 = LinearInterpolate(
   A=LinearInterpolate(
    A=LinearInterpolate(B=@n33, Alpha=@n34)
    B=ScalarParameter() {defaultValue=129.0307, parameterName=Foot_Radius_stage2}
    Alpha=NamedRerouteUsage()
   )
   B=ScalarParameter() {defaultValue=173.5603, parameterName=Foot_Radius_stage3}
   Alpha=NamedRerouteUsage()
  )
  n36 = ScalarParameter() {defaultValue=0.97, parameterName=Foot_Hardness}
  n37 = Max(
   A=SphereMask(
    A=Add(A=@n13, B=@n14)
    B=Constant()
    Radius=LinearInterpolate(A=@n15, B=@n16, Alpha=@n17)
    Hardness=ScalarParameter() {defaultValue=0.97, parameterName=Leg_R_Hardness}
   )
   B=Max(A=SphereMask(A=@n22, B=@n23, Radius=@n26, Hardness=@n27), B=SphereMask(A=@n31, B=@n32, Radius=@n35, Hardness=@n36))
  )
  n38 = NamedRerouteDeclaration(None=Max(A=SphereMask(A=@n2, B=@n3, Radius=@n4, Hardness=@n5), B=Max(A=@n11, B=@n37))) {name=OutMask, nodeColor=<rgba 0 0.5547 1 1>}
  n39 = ScalarParameter() {defaultValue=1, parameterName=Global Mask Intensity}
  FunctionOutput(None=LinearInterpolate(B=Multiply(A=@n38, B=@n39), Alpha=ScalarParameter() {defaultValue=1, parameterName=UseReactive}) {constA=1}) {outputName=OutMask}
  n40 = ScalarParameter() {defaultValue=1, parameterName=SpeedOffset}
  n41 = Multiply(A=ScalarParameter() {group=SheerWill_BlendMask, parameterName=Stage2_3}, B=@n40)
  FunctionOutput(None=NamedRerouteDeclaration(None=Saturate(None=@n41)) {name=Animation_Stage2_3, nodeColor=<rgba 0.7344 0 1 1>}) {outputName=Stage2_3}
  n42 = Multiply(A=ScalarParameter() {group=SheerWill_BlendMask, parameterName=Stage3_4}, B=@n40)
  FunctionOutput(None=NamedRerouteDeclaration(None=Saturate(None=@n42)) {name=Animation_Stage3_4, nodeColor=<rgba 0.7344 0 1 1>}) {outputName=Stage3_4}
  n43 = Multiply(A=ScalarParameter() {group=SheerWill_BlendMask, parameterName=Stage0_1}, B=@n40)
  FunctionOutput(None=NamedRerouteDeclaration(None=Saturate(None=@n43)) {name=Animation_Stage0_1, nodeColor=<rgba 0.7344 0 1 1>}) {outputName=Stage0_1}
  n44 = Multiply(A=ScalarParameter() {group=SheerWill_BlendMask, parameterName=Stage1_2}, B=@n40)
  FunctionOutput(None=NamedRerouteDeclaration(None=Saturate(None=@n44)) {name=Animation_Stage1_2, nodeColor=<rgba 0.7344 0 1 1>}) {outputName=Stage1_2}
  NamedRerouteDeclaration(None=VertexInterpolator(VS=LocalPosition() {localOrigin=InstancePreSkinning}.XYZ).PS) {name=PreSkinnedLocalPosition, nodeColor=<rgba 0 1 0.0625 1>}
 ]}
