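# Parent body material for the SheerWill FrothRain skin, layering an animated crystal effect onto the shared 2023 character PreFX/PostFX functions.
# MF_SheerWill_Mask drives the PreFX blend and four stage values, whose sine pulses modulate edge glow.
# A crystal branch built on MF_Crystal adds curvature/Fresnel emissive and UseReactive-controlled vertex offset relative to OrbLocation.
# MF_SheerWill_EmoteMask blends in a black, rough layer with glowing edges before PostFX.
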Material /BRCosmetics/Characters/Player/Male/Medium/Bodies/M_MED_SheerWill_Blot/Skins/FrothRain/Materials/M_SheerWill_FrothRain_Body_Parent
graph {
 properties {
  bAutomaticallySetUsageInEditor=false
  bUseMaterialAttributes=true
  bUsedWithMeshDeformer=true
  bUsedWithMorphTargets=true
  bUsedWithSkeletalMesh=true
  pixelDepthOffsetMode=PDOM_Legacy
  shadingModel=MSM_FromMaterialExpression
  usedShadingModels=Default Lit | Subsurface | Clear Coat | Subsurface Profile | Cloth
 }
 nodes [
  n1 = ScalarParameter() {defaultValue=2, parameterName=Crack_CRV_intensityMax}
  n2 = Multiply(A=TextureCoordinate(), B=ScalarParameter() {defaultValue=5, parameterName=Crack_tiling})
  LinearInterpolate(
   A=ScalarParameter() {defaultValue=0.3, parameterName=Crack_CRV_intensityMin}
   B=@n1
   Alpha=TextureSample(UVs=Panner(Coordinate=@n2) {speedY=0.04}) {samplerType=SAMPLERTYPE_Grayscale, texture=/Game/Global/Textures/Default/Noise/LowResBlurredNoiseNormalized}.R
  ) {constA=0.5}
  n3 = MaterialFunctionCall() {materialFunction=/BRCosmetics/Characters/Player/Male/Medium/Bodies/M_MED_SheerWill_Blot/Skins/FrothRain/Materials/MF/MF_SheerWill_Mask}
  n4 = MaterialFunctionCall!(Overide Texture Set UVs, In_BlendMaskAnimation=@n3.OutMask, Override Tech Art Mask UVs, Override Basic Emissive UVs) {materialFunction=/BRCosmetics/Characters/Player/Male/Medium/Bodies/M_MED_SheerWill_Blot/Skins/FrothRain/Materials/MF/MF_CharacterParent_2023_PreFX_SheerWil}
  n5 = GetMaterialAttributes(None=NamedRerouteDeclaration(None=@n4.Result) {name=IN_MA, nodeColor=<rgba 0 1 0.6016 1>}) {attributeGetTypes [#1, #2]}
  n6 = Constant() {r=0.6143}
  n7 = Constant()
  n8 = NamedRerouteUsage()
  n9 = Multiply(A=Multiply(A=SmoothStep(Min=@n6, Max=@n7, Value=@n8), B=@n8), B=ScalarParameter() {defaultValue=5, parameterName=GlowEdgeIntensity}) {constB=0.5}
  n10 = NamedRerouteDeclaration(None=@n3.Stage2_3) {name=In_Stage2_3, nodeColor=<rgba 1 0 0.0469 1>}
  n11 = MaterialFunctionCall() {materialFunction=&Pi}
  n12 = NamedRerouteDeclaration(None=Sine(None=Multiply(A=@n10, B=@n11.Result)) {period=6.283}) {name=Sine_Stage2_3, nodeColor=<rgba 0 1 0.6016 1>}
  n13 = NamedRerouteDeclaration(None=@n3.Stage3_4) {name=In_Stage3_4, nodeColor=<rgba 0 0.3672 1 1>}
  n14 = MaterialFunctionCall() {materialFunction=&Pi}
  n15 = NamedRerouteDeclaration(None=Sine(None=Multiply(A=@n13, B=@n14.Result)) {period=6.283}) {name=Sine_Stage3_4, nodeColor=<rgba 0 1 0.6016 1>}
  n16 = Sine(None=Multiply(
    A=NamedRerouteDeclaration(None=@n3.Stage0_1) {name=In_Stage0_1, nodeColor=<rgba 0 0.3672 1 1>}
    B=MaterialFunctionCall() {materialFunction=&Pi}.Result
   )) {period=6.283}
  n17 = Max(
   A=Max(A=Saturate(None=@n12), B=Saturate(None=@n15))
   B=Saturate(None=NamedRerouteDeclaration(None=@n16) {name=Sine_Stage0_1, nodeColor=<rgba 0 1 0.6016 1>})
  )
  n18 = Multiply(
   A=NamedRerouteDeclaration(None=@n3.Stage1_2) {name=In_Stage1_2, nodeColor=<rgba 0 0.3672 1 1>}
   B=MaterialFunctionCall() {materialFunction=&Pi}.Result
  )
  n19 = Saturate(None=NamedRerouteDeclaration(None=Sine(None=@n18) {period=6.283}) {name=Sine_Stage1_2, nodeColor=<rgba 0 1 0.6016 1>})
  n20 = LinearInterpolate(A=ScalarParameter() {defaultValue=0.015, parameterName=GlowEdgeIntensityMIn}, Alpha=Power(Base=Max(A=@n17, B=@n19))) {constA=0.1}
  n21 = BlendMaterialAttributes(
   A=SetMaterialAttributes!(MaterialAttributes=@n5.MaterialAttributes, Base Color=@n5.MaterialAttributes, Emissive Color=Multiply(A=@n9, B=@n20)) {attributeSetTypes [#1, #2]}
   B=NamedRerouteUsage()
   Alpha=NamedRerouteUsage()
  )
  n22 = MaterialFunctionCall() {materialFunction=/BRCosmetics/Characters/Player/Male/Medium/Bodies/M_MED_SheerWill_Blot/Skins/FrothRain/Materials/MF/MF_SheerWill_EmoteMask}
  n23 = SetMaterialAttributes!(
   MaterialAttributes=@n21
   Base Color=Constant()
   Emissive Color=@n22.EmoteEdgeColor
   Metallic
   Specular=Constant() {r=0.1}
   Roughness=Constant() {r=1}
  ) {attributeSetTypes [
    #1
    #2
    #3
    #4
    #5
   ]}
  MaterialFunctionCall(
   Material=BlendMaterialAttributes(A=@n21, B=@n23, Alpha=@n22.OutMask)
   M=@n4.M
   Pre Skinned Local Position=@n4.M
   Distance Blend=@n4.M
   Fresnel=@n4.M
  ) {materialFunction=&MF_CharacterParent_2023_PostFX}
  NamedRerouteDeclaration(None=SphereMask(A=VertexColor().R, B=Constant() {r=0.1}) {attenuationRadius=0.01}) {name=Crystal mask, nodeColor=<rgba 0 0.4375 1 1>}
  n24 = GetMaterialAttributes(None=MaterialFunctionCall!(Material=NamedRerouteUsage(), Alpha=NamedRerouteUsage()) {materialFunction=/Game/Global/MaterialFunctions/TechArt/MF_Crystal}.Result) {attributeGetTypes [
    #5
    #3
    #4
    #6
    #1
    #2
    #7
   ]}
  NamedRerouteDeclaration(None=@n24.Emissive Color) {name=In_EmissiveColor, nodeColor=<rgba 0 0.7656 1 1>}
  NamedRerouteDeclaration(None=@n3.OutMask) {name=In_Mask, nodeColor=<rgba 0 1 0.3203 1>}
  n25 = TextureSample() {
   samplerSource=SSM_Wrap_WorldGroupSettings
   samplerType=SAMPLERTYPE_Masks
   texture=/BRCosmetics/Characters/Player/Male/Medium/Bodies/M_MED_SheerWill_Blot/Skins/FrothRain/Textures/T_SheerWill_Blot_Body_CRV
  }
  NamedRerouteDeclaration(None=@n25.B) {name=Crystal mask3, nodeColor=<rgba 0.4062 1 0 1>}
  n26 = TextureSampleParameter2D() {
   group=00_Texture Inputs
   parameterName=Diffuse
   samplerSource=SSM_Wrap_WorldGroupSettings
   sortPriority=0
   texture=/Game/Global/Textures/Default/Blanks/T_MidgreysRGB
  }
  n27 = MaterialFunctionCall(In=@n24.Base Color, Contrast=ScalarParameter() {defaultValue=0.002, parameterName=Crystal_cheapContrast}) {materialFunction=&CheapContrast}
  n28 = Constant() {r=1}
  n29 = ScalarParameter() {defaultValue=1, parameterName=UseReactive}
  n30 = LinearInterpolate(B=If(A=NamedRerouteUsage(), "A > B"=@n28, "A == B"=@n28, "A < B"=Constant()) {constB=0.9}, Alpha=@n29) {constA=1}
  n31 = SetMaterialAttributes!(
   MaterialAttributes=@n24.MaterialAttributes
   Roughness=ScalarParameter() {defaultValue=0.01, parameterName=Crystal_roughness}
   Metallic=ScalarParameter() {defaultValue=0.25, parameterName=Crystal_Metallic}
   Specular=ScalarParameter() {defaultValue=1, parameterName=Crystal_Spec}
   Refraction=@n24.MaterialAttributes
   Normal
   Base Color=Multiply(A=LinearInterpolate(A=@n26.RGB, B=@n27.Result, Alpha=@n30), B=Constant() {r=2})
   Emissive Color=NamedRerouteUsage()
   World Position Offset=LinearInterpolate(B=NamedRerouteUsage(), Alpha=@n29)
  ) {attributeSetTypes [
    #5
    #3
    #4
    #7
    #6
    #1
    #2
    #8
   ]}
  NamedRerouteDeclaration(None=@n31) {name=CrystalMA, nodeColor=<rgba 1 0.2812 0 1>}
  n32 = Subtract(A=VectorParameter() {parameterName=OrbLocation, defaultValue=<rgba -0.312 13.8193 131.7849 1>}.RGB, B=WorldPosition().XYZ)
  n33 = NamedRerouteUsage()
  n34 = Saturate(None=Multiply(A=NamedRerouteUsage()) {constB=2})
  n35 = Multiply(A=NamedRerouteUsage()) {constB=2}
  n36 = Multiply(A=LinearInterpolate(A=LinearInterpolate(B=@n33, Alpha=@n34), Alpha=Saturate(None=@n35)), B=NamedRerouteUsage())
  NamedRerouteDeclaration(None=LinearInterpolate(B=LinearInterpolate(A=@n32, Alpha=@n36) {constB=0}, Alpha=NamedRerouteUsage())) {name=WPO, nodeColor=<rgba 0.875 0 1 1>}
  n37 = ScalarParameter() {defaultValue=2.2, group=SheerWiill, parameterName=FresnelEdge_Exponentln}
  n38 = Multiply(
   A=Multiply(
    A=ScalarParameter() {defaultValue=2, group=SheerWiill, parameterName=FresnelEdgeIntensity}
    B=VectorParameter() {group=SheerWiill, parameterName=FresnelEdgeColor, defaultValue=<rgba 0 0 0 1>}.RGB
   )
   B=Power(Base=Fresnel(ExponentIn=@n37))
  )
  n39 = Add(A=Multiply(A=@n1, B=@n25.R) {constB=0.1}, B=NamedRerouteUsage())
  n40 = Constant() {r=1.5}
  n41 = Fresnel(ExponentIn=ScalarParameter() {defaultValue=0.6612, group=Emissive, parameterName=Emissive_FresnelExponentln})
  n42 = Multiply(
   A=Saturate(None=Multiply(A=@n39, B=@n40))
   B=LinearInterpolate(B=ScalarParameter() {defaultValue=4, group=Emissive, parameterName=Emissive_FresnelPower}, Alpha=Power(Base=@n41)) {constB=2}
  )
  NamedRerouteDeclaration(None=Multiply(A=Add(A=@n38, B=@n42), B=Constant() {r=0.5}) {constB=0.5}) {name=OutCrystalEmissiveColor, nodeColor=<rgba 0.8984 0 1 1>}
  Saturate(None=Subtract(A=NamedRerouteDeclaration(None=@n24.Roughness) {name=In_Roughness, nodeColor=<rgba 0 1 0.2734 1>}, B=Constant() {r=0.1}))
  ScalarParameter() {defaultValue=0.02, parameterName=Crack_intensity}
 ]
}
